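# Physics asset for the cape on the male medium Candor player body, a variant named for motorbike use.
# Capsules on pelvis, spine and clavicles form the torso shape, while arm and leg bodies carry no shapes.
# Six four-segment cape chains hang from base bodies on upper arms, clavicles and spine_05, with angle-limited joints along each.
# Angularly free cross-constraints tie left mid and back chains at segments 02–04, with no right-side equivalent.
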PhysicsAsset /BRCosmetics/Characters/Player/Male/Medium/Bodies/M_MED_Candor/Meshes/M_MED_Candor_Cape_Motorbike_Physics
bodies {
 calf_l []
 calf_r []
 clavicle_l [{shapeType=Capsule, length=7.7636, radius=11.1427, center=<9.9642 -0.7741 -1.5923>, rotation=<rot 75.9352 -0.0001 175.5622>}]
 clavicle_r [{shapeType=Capsule, length=7.7636, radius=11.1427, center=<-9.9642 0.7741 1.5923>, rotation=<rot 75.9351 -0.0001 -4.4378>}]
 dyn_cape_01_l [{shapeType=Capsule, length=16, radius=3.5, shapeName=dyn_cape_01_r_capsule, center=<4 0 0>}]
 dyn_cape_01_r [{shapeType=Capsule, length=15, radius=4.5, shapeName=dyn_cape_01_r_capsule, center=<4 0 0>}]
 dyn_cape_02_l [{shapeType=Capsule, length=16, radius=3.5, shapeName=dyn_cape_01_r_capsule, center=<4 0 0>}]
 dyn_cape_02_r [{shapeType=Capsule, length=15, radius=4.5, shapeName=dyn_cape_01_r_capsule, center=<4 0 0>}]
 dyn_cape_03_l [{shapeType=Capsule, length=26, radius=3.5, shapeName=dyn_cape_01_r_capsule, center=<4 0 0>}]
 dyn_cape_03_r [{shapeType=Capsule, length=15, radius=4.5, shapeName=dyn_cape_01_r_capsule, center=<4 0 0>}]
 dyn_cape_04_l [{shapeType=Capsule, length=24.75, radius=4.75, shapeName=dyn_cape_01_r_capsule, center=<4 0 0>}]
 dyn_cape_04_r [{shapeType=Capsule, length=13.75, radius=5.75, shapeName=dyn_cape_01_r_capsule, center=<4 0 0>}]
 dyn_cape_back_01_l [{shapeType=Capsule, length=13.25, radius=3, shapeName=dyn_cape_01_r_capsule, center=<5.7779 0 0>}]
 dyn_cape_back_01_r [{shapeType=Capsule, length=10.25, radius=4, shapeName=dyn_cape_01_r_capsule, center=<4 0 0>}]
 dyn_cape_back_02_l [{shapeType=Capsule, length=16.75, radius=2.75, shapeName=dyn_cape_01_r_capsule, center=<5.7779 0 0>}]
 dyn_cape_back_02_r [{shapeType=Capsule, length=15.75, radius=3.75, shapeName=dyn_cape_01_r_capsule, center=<4 0 0>}]
 dyn_cape_back_03_l [{shapeType=Capsule, length=26, radius=3.5, shapeName=dyn_cape_01_r_capsule, center=<5.7779 0 0>}]
 dyn_cape_back_03_r [{shapeType=Capsule, length=15, radius=4.5, shapeName=dyn_cape_01_r_capsule, center=<4 0 0>}]
 dyn_cape_back_04_l [{shapeType=Capsule, length=24.75, radius=4.75, shapeName=dyn_cape_01_r_capsule, center=<5.7779 0 0>}]
 dyn_cape_back_04_r [{shapeType=Capsule, length=13.75, radius=5.75, shapeName=dyn_cape_01_r_capsule, center=<4 0 0>}]
 dyn_cape_back_l_base [{shapeType=Capsule, length=0.505, radius=0.505, shapeName=dyn_cape_back_l_base_capsule, rotation=<rot 0 -90 0>}]
 dyn_cape_back_r_base [{shapeType=Capsule, length=0.505, radius=0.505, shapeName=dyn_cape_back_r_base_capsule, rotation=<rot 0 -90 0>}]
 dyn_cape_l_base [{shapeType=Capsule, length=0.505, radius=0.505, shapeName=dyn_cape_l_base_capsule, rotation=<rot 0 -90 0>}]
 dyn_cape_mid_01_l [{shapeType=Capsule, length=10.5, radius=3.75, shapeName=dyn_cape_01_r_capsule, center=<4.6752 0 0>}]
 dyn_cape_mid_01_r [{shapeType=Capsule, length=8.25, radius=6, shapeName=dyn_cape_01_r_capsule, center=<4 0 0>}]
 dyn_cape_mid_02_l [{shapeType=Capsule, length=15.75, radius=3.75, shapeName=dyn_cape_01_r_capsule, center=<4.6752 0 0>}]
 dyn_cape_mid_02_r [{shapeType=Capsule, length=13.75, radius=5.75, shapeName=dyn_cape_01_r_capsule, center=<4 0 0>}]
 dyn_cape_mid_03_l [{shapeType=Capsule, length=25, radius=4.5, shapeName=dyn_cape_01_r_capsule, center=<4 0 0>}]
 dyn_cape_mid_03_r [{shapeType=Capsule, length=15, radius=4.5, shapeName=dyn_cape_01_r_capsule, center=<4 0 0>}]
 dyn_cape_mid_04_l [{shapeType=Capsule, length=23.75, radius=5.75, shapeName=dyn_cape_01_r_capsule, center=<4 0 0>}]
 dyn_cape_mid_04_r [{shapeType=Capsule, length=13.75, radius=5.75, shapeName=dyn_cape_01_r_capsule, center=<4 0 0>}]
 dyn_cape_mid_l_base [{shapeType=Capsule, length=0.505, radius=0.505, shapeName=dyn_cape_mid_l_base_capsule, rotation=<rot 0 -90 0>}]
 dyn_cape_mid_r_base [{shapeType=Capsule, length=0.505, radius=0.505, shapeName=dyn_cape_mid_r_base_capsule, rotation=<rot 0 -90 0>}]
 dyn_cape_r_base [{shapeType=Capsule, length=0.505, radius=0.505, shapeName=dyn_cape_r_base_capsule, rotation=<rot 0 -90 0>}]
 lowerarm_l []
 lowerarm_r []
 pelvis [{shapeType=Capsule, length=26.9062, radius=12.2888, center=<-8.5193 -1.3695 0>}]
 spine_01 [{shapeType=Capsule, length=14.2187, radius=11.1638, center=<-2.6215 -3.2711 0>}]
 spine_02 [
  {shapeType=Capsule, length=15.8437, radius=9.945, center=<-2.6215 0.4993 0.0001>}
  {shapeType=Capsule, length=29.875, radius=11.1013, center=<2.5024 0.4992 3.9825>, rotation=<rot 90 0.0049 0.0049>}
  {shapeType=Capsule, length=29.875, radius=11.1013, center=<2.4704 0.4991 -3.9825>, rotation=<rot 90 -0.0396 -0.0396>}
 ]
 spine_03 [{shapeType=Capsule, length=10.4062, radius=11.7575, center=<-1.953 0.1364 0.0001>}]
 spine_04 [{shapeType=Capsule, length=9.6562, radius=10.3825, center=<2.1467 -1.2748 0.0002>}]
 spine_05 [{shapeType=Capsule, length=17.8437, radius=17.195, center=<-1.7514 2.4879 0>}]
 thigh_l []
 thigh_r []
 upperarm_l []
 upperarm_r []
}
constraints [
 {
  bone1Name=spine_02
  bone2Name=spine_01
  swing1LimitDegrees=45
  swing1Motion=Free
  swing2LimitDegrees=45
  swing2Motion=Free
  twistLimitDegrees=45
  twistMotion=Free
 }
 {
  bone1Name=spine_03
  bone2Name=spine_02
  swing1LimitDegrees=45
  swing1Motion=Free
  swing2LimitDegrees=45
  swing2Motion=Free
  twistLimitDegrees=45
  twistMotion=Free
 }
 {
  bone1Name=spine_04
  bone2Name=spine_03
  swing1LimitDegrees=45
  swing1Motion=Free
  swing2LimitDegrees=45
  swing2Motion=Free
  twistLimitDegrees=45
  twistMotion=Free
 }
 {
  bone1Name=spine_01
  bone2Name=pelvis
  swing1LimitDegrees=45
  swing1Motion=Free
  swing2LimitDegrees=45
  swing2Motion=Free
  twistLimitDegrees=45
  twistMotion=Free
 }
 {
  bone1Name=spine_05
  bone2Name=spine_04
  swing1LimitDegrees=45
  swing1Motion=Free
  swing2LimitDegrees=45
  swing2Motion=Free
  twistLimitDegrees=45
  twistMotion=Free
 }
 {
  bone1Name=upperarm_l
  bone2Name=clavicle_l
  swing1LimitDegrees=45
  swing1Motion=Free
  swing2LimitDegrees=45
  swing2Motion=Free
  twistLimitDegrees=45
  twistMotion=Free
 }
 {
  bone1Name=clavicle_l
  bone2Name=spine_05
  swing1LimitDegrees=45
  swing1Motion=Free
  swing2LimitDegrees=45
  swing2Motion=Free
  twistLimitDegrees=45
  twistMotion=Free
 }
 {
  bone1Name=lowerarm_l
  bone2Name=upperarm_l
  swing1LimitDegrees=45
  swing1Motion=Free
  swing2LimitDegrees=45
  swing2Motion=Free
  twistLimitDegrees=45
  twistMotion=Free
 }
 {
  bone1Name=clavicle_r
  bone2Name=spine_05
  swing1LimitDegrees=45
  swing1Motion=Free
  swing2LimitDegrees=45
  swing2Motion=Free
  twistLimitDegrees=45
  twistMotion=Free
 }
 {
  bone1Name=upperarm_r
  bone2Name=clavicle_r
  swing1LimitDegrees=45
  swing1Motion=Free
  swing2LimitDegrees=45
  swing2Motion=Free
  twistLimitDegrees=45
  twistMotion=Free
 }
 {
  bone1Name=lowerarm_r
  bone2Name=upperarm_r
  swing1LimitDegrees=45
  swing1Motion=Free
  swing2LimitDegrees=45
  swing2Motion=Free
  twistLimitDegrees=45
  twistMotion=Free
 }
 {
  bone1Name=thigh_l
  bone2Name=pelvis
  swing1LimitDegrees=45
  swing1Motion=Free
  swing2LimitDegrees=45
  swing2Motion=Free
  twistLimitDegrees=45
  twistMotion=Free
 }
 {
  bone1Name=thigh_r
  bone2Name=pelvis
  swing1LimitDegrees=45
  swing1Motion=Free
  swing2LimitDegrees=45
  swing2Motion=Free
  twistLimitDegrees=45
  twistMotion=Free
 }
 {
  bone1Name=calf_l
  bone2Name=thigh_l
  swing1LimitDegrees=45
  swing1Motion=Free
  swing2LimitDegrees=45
  swing2Motion=Free
  twistLimitDegrees=45
  twistMotion=Free
 }
 {
  bone1Name=calf_r
  bone2Name=thigh_r
  swing1LimitDegrees=45
  swing1Motion=Free
  swing2LimitDegrees=45
  swing2Motion=Free
  twistLimitDegrees=45
  twistMotion=Free
 }
 {
  bone1Name=dyn_cape_mid_l_base
  bone2Name=clavicle_l
  swing1LimitDegrees=45
  swing1Motion=Free
  swing2LimitDegrees=45
  swing2Motion=Free
  twistLimitDegrees=45
  twistMotion=Free
 }
 {bone1Name=dyn_cape_mid_01_l, bone2Name=dyn_cape_mid_l_base, swing1LimitDegrees=25, swing2LimitDegrees=15, twistLimitDegrees=5}
 {bone1Name=dyn_cape_mid_02_l, bone2Name=dyn_cape_mid_01_l, swing1LimitDegrees=25, swing2LimitDegrees=15, twistLimitDegrees=5}
 {bone1Name=dyn_cape_mid_03_l, bone2Name=dyn_cape_mid_02_l, swing1LimitDegrees=15, swing2LimitDegrees=15, twistLimitDegrees=5}
 {bone1Name=dyn_cape_mid_04_l, bone2Name=dyn_cape_mid_03_l, swing1LimitDegrees=15, swing2LimitDegrees=15, twistLimitDegrees=5}
 {
  bone1Name=dyn_cape_l_base
  bone2Name=upperarm_l
  swing1LimitDegrees=45
  swing1Motion=Free
  swing2LimitDegrees=45
  swing2Motion=Free
  twistLimitDegrees=45
  twistMotion=Free
 }
 {bone1Name=dyn_cape_01_l, bone2Name=dyn_cape_l_base, swing1LimitDegrees=25, swing2LimitDegrees=30, twistLimitDegrees=5}
 {bone1Name=dyn_cape_02_l, bone2Name=dyn_cape_01_l, swing1LimitDegrees=15, swing2LimitDegrees=15, twistLimitDegrees=5}
 {bone1Name=dyn_cape_03_l, bone2Name=dyn_cape_02_l, swing1LimitDegrees=15, swing2LimitDegrees=15, twistLimitDegrees=5}
 {bone1Name=dyn_cape_04_l, bone2Name=dyn_cape_03_l, swing1LimitDegrees=15, swing2LimitDegrees=15, twistLimitDegrees=5}
 {
  bone1Name=dyn_cape_mid_r_base
  bone2Name=clavicle_r
  swing1LimitDegrees=45
  swing1Motion=Free
  swing2LimitDegrees=45
  swing2Motion=Free
  twistLimitDegrees=45
  twistMotion=Free
 }
 {bone1Name=dyn_cape_mid_01_r, bone2Name=dyn_cape_mid_r_base, swing1LimitDegrees=15, swing2LimitDegrees=15, twistLimitDegrees=15}
 {bone1Name=dyn_cape_mid_02_r, bone2Name=dyn_cape_mid_01_r, swing1LimitDegrees=15, swing2LimitDegrees=15, twistLimitDegrees=15}
 {bone1Name=dyn_cape_mid_03_r, bone2Name=dyn_cape_mid_02_r, swing1LimitDegrees=15, swing2LimitDegrees=15, twistLimitDegrees=15}
 {bone1Name=dyn_cape_mid_04_r, bone2Name=dyn_cape_mid_03_r, swing1LimitDegrees=15, swing2LimitDegrees=15, twistLimitDegrees=15}
 {
  bone1Name=dyn_cape_r_base
  bone2Name=upperarm_r
  swing1LimitDegrees=45
  swing1Motion=Free
  swing2LimitDegrees=45
  swing2Motion=Free
  twistLimitDegrees=45
  twistMotion=Free
 }
 {bone1Name=dyn_cape_01_r, bone2Name=dyn_cape_r_base, swing1LimitDegrees=25, swing2LimitDegrees=30, twistLimitDegrees=15}
 {bone1Name=dyn_cape_02_r, bone2Name=dyn_cape_01_r, swing1LimitDegrees=25, swing2LimitDegrees=15, twistLimitDegrees=15}
 {bone1Name=dyn_cape_03_r, bone2Name=dyn_cape_02_r, swing1LimitDegrees=25, swing2LimitDegrees=15, twistLimitDegrees=15}
 {bone1Name=dyn_cape_04_r, bone2Name=dyn_cape_03_r, swing1LimitDegrees=25, swing2LimitDegrees=15, twistLimitDegrees=15}
 {
  bone1Name=dyn_cape_back_l_base
  bone2Name=spine_05
  swing1LimitDegrees=45
  swing1Motion=Free
  swing2LimitDegrees=45
  swing2Motion=Free
  twistLimitDegrees=45
  twistMotion=Free
 }
 {bone1Name=dyn_cape_back_01_l, bone2Name=dyn_cape_back_l_base, swing1LimitDegrees=15, swing2LimitDegrees=15, twistLimitDegrees=5}
 {bone1Name=dyn_cape_back_02_l, bone2Name=dyn_cape_back_01_l, swing1LimitDegrees=15, swing2LimitDegrees=15, twistLimitDegrees=5}
 {bone1Name=dyn_cape_back_03_l, bone2Name=dyn_cape_back_02_l, swing1LimitDegrees=15, swing2LimitDegrees=15, twistLimitDegrees=5}
 {bone1Name=dyn_cape_back_04_l, bone2Name=dyn_cape_back_03_l, swing1LimitDegrees=15, swing2LimitDegrees=15, twistLimitDegrees=5}
 {
  bone1Name=dyn_cape_back_r_base
  bone2Name=spine_05
  swing1LimitDegrees=45
  swing1Motion=Free
  swing2LimitDegrees=45
  swing2Motion=Free
  twistLimitDegrees=45
  twistMotion=Free
 }
 {bone1Name=dyn_cape_back_01_r, bone2Name=dyn_cape_back_r_base, swing1LimitDegrees=15, swing2LimitDegrees=15, twistLimitDegrees=15}
 {bone1Name=dyn_cape_back_02_r, bone2Name=dyn_cape_back_01_r, swing1LimitDegrees=15, swing2LimitDegrees=15, twistLimitDegrees=15}
 {bone1Name=dyn_cape_back_03_r, bone2Name=dyn_cape_back_02_r, swing1LimitDegrees=15, swing2LimitDegrees=15, twistLimitDegrees=15}
 {bone1Name=dyn_cape_back_04_r, bone2Name=dyn_cape_back_03_r, swing1LimitDegrees=15, swing2LimitDegrees=15, twistLimitDegrees=15}
 {
  bone1Name=dyn_cape_mid_04_l
  bone2Name=dyn_cape_back_04_l
  swing1LimitDegrees=45
  swing1Motion=Free
  swing2LimitDegrees=45
  swing2Motion=Free
  twistLimitDegrees=45
  twistMotion=Free
 }
 {
  bone1Name=dyn_cape_mid_03_l
  bone2Name=dyn_cape_back_03_l
  swing1LimitDegrees=45
  swing1Motion=Free
  swing2LimitDegrees=45
  swing2Motion=Free
  twistLimitDegrees=45
  twistMotion=Free
 }
 {
  bone1Name=dyn_cape_mid_02_l
  bone2Name=dyn_cape_back_02_l
  swing1LimitDegrees=45
  swing1Motion=Free
  swing2LimitDegrees=45
  swing2Motion=Free
  twistLimitDegrees=45
  twistMotion=Free
 }
]
