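# Physics asset for the Candor medium male body's frontend cape, giving the cape bone chains collision and joint limits.
# Six chains (side, mid and back, left and right) each run four capsules from a tiny base body, with tight swing and twist limits.
# Side bases attach to the upper arms and mid and back bases to spine_05, with free rotation.
# Capsules cover pelvis and spine, while arm and leg bodies have no shapes.
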PhysicsAsset /BRCosmetics/Characters/Player/Male/Medium/Bodies/M_MED_Candor/Meshes/M_MED_Candor_Frontend_Cape_Physics
bodies {
 calf_l []
 calf_r []
 dyn_cape_01_l [{shapeType=Capsule, length=16, radius=3.5, shapeName=dyn_cape_01_r_capsule, center=<4 0 0>}]
 dyn_cape_01_r [{shapeType=Capsule, length=15, radius=4.5, shapeName=dyn_cape_01_r_capsule, center=<4 0 0>}]
 dyn_cape_02_l [{shapeType=Capsule, length=16, radius=3.5, shapeName=dyn_cape_01_r_capsule, center=<4 0 0>}]
 dyn_cape_02_r [{shapeType=Capsule, length=15, radius=4.5, shapeName=dyn_cape_01_r_capsule, center=<4 0 0>}]
 dyn_cape_03_l [{shapeType=Capsule, length=16, radius=3.5, shapeName=dyn_cape_01_r_capsule, center=<4 0 0>}]
 dyn_cape_03_r [{shapeType=Capsule, length=16.5, radius=3, shapeName=dyn_cape_01_r_capsule, center=<4 0 0>}]
 dyn_cape_04_l [{shapeType=Capsule, length=14.75, radius=4.75, shapeName=dyn_cape_01_r_capsule, center=<4 0 0>}]
 dyn_cape_04_r [{shapeType=Capsule, length=14.75, radius=4.75, shapeName=dyn_cape_01_r_capsule, center=<4 0 0>}]
 dyn_cape_back_01_l [{shapeType=Capsule, length=13.25, radius=3, shapeName=dyn_cape_01_r_capsule, center=<5.7779 0 0>}]
 dyn_cape_back_01_r [{shapeType=Capsule, length=13.25, radius=3, shapeName=dyn_cape_01_r_capsule, center=<5.7779 0 0>}]
 dyn_cape_back_02_l [{shapeType=Capsule, length=16.75, radius=5, shapeName=dyn_cape_01_r_capsule, center=<5.7779 0 0>}]
 dyn_cape_back_02_r [{shapeType=Capsule, length=16.75, radius=5, shapeName=dyn_cape_01_r_capsule, center=<5.7779 0 0>}]
 dyn_cape_back_03_l [{shapeType=Capsule, length=16, radius=5, shapeName=dyn_cape_01_r_capsule, center=<5.7779 0 0>}]
 dyn_cape_back_03_r [{shapeType=Capsule, length=14.5, radius=5, shapeName=dyn_cape_01_r_capsule, center=<5.7779 0 0>}]
 dyn_cape_back_04_l [{shapeType=Capsule, length=14.75, radius=5, shapeName=dyn_cape_01_r_capsule, center=<5.7779 0 0>}]
 dyn_cape_back_04_r [{shapeType=Capsule, length=14.75, radius=5, shapeName=dyn_cape_01_r_capsule, center=<5.7779 0 0>}]
 dyn_cape_back_l_base [{shapeType=Capsule, length=0.505, radius=0.505, shapeName=dyn_cape_back_l_base_capsule, rotation=<rot 0 -90 0>}]
 dyn_cape_back_r_base [{shapeType=Capsule, length=0.505, radius=0.505, shapeName=dyn_cape_back_r_base_capsule, rotation=<rot 0 -90 0>}]
 dyn_cape_l_base [{shapeType=Capsule, length=0.505, radius=0.505, shapeName=dyn_cape_l_base_capsule, rotation=<rot 0 -90 0>}]
 dyn_cape_mid_01_l [{shapeType=Capsule, length=10.5, radius=3.75, shapeName=dyn_cape_01_r_capsule, center=<4.6752 0 0>}]
 dyn_cape_mid_01_r [{shapeType=Capsule, length=8.25, radius=6, shapeName=dyn_cape_01_r_capsule, center=<4 0 0>}]
 dyn_cape_mid_02_l [{shapeType=Capsule, length=15.75, radius=5, shapeName=dyn_cape_01_r_capsule, center=<4.6752 0 0>}]
 dyn_cape_mid_02_r [{shapeType=Capsule, length=13.75, radius=5, shapeName=dyn_cape_01_r_capsule, center=<4 0 0>}]
 dyn_cape_mid_03_l [{shapeType=Capsule, length=15, radius=5, shapeName=dyn_cape_01_r_capsule, center=<4 0 0>}]
 dyn_cape_mid_03_r [{shapeType=Capsule, length=16.5, radius=5, shapeName=dyn_cape_01_r_capsule, center=<4 0 0>}]
 dyn_cape_mid_04_l [{shapeType=Capsule, length=13.75, radius=5, shapeName=dyn_cape_01_r_capsule, center=<4 0 0>}]
 dyn_cape_mid_04_r [{shapeType=Capsule, length=13.75, radius=5, shapeName=dyn_cape_01_r_capsule, center=<4 0 0>}]
 dyn_cape_mid_l_base [{shapeType=Capsule, length=0.505, radius=0.505, shapeName=dyn_cape_mid_l_base_capsule, rotation=<rot 0 -90 0>}]
 dyn_cape_mid_r_base [{shapeType=Capsule, length=0.505, radius=0.505, shapeName=dyn_cape_mid_r_base_capsule, rotation=<rot 0 -90 0>}]
 dyn_cape_r_base [{shapeType=Capsule, length=0.505, radius=0.505, shapeName=dyn_cape_r_base_capsule, rotation=<rot 0 -90 0>}]
 lowerarm_l []
 lowerarm_r []
 pelvis [{shapeType=Capsule, length=19.9062, radius=10.7888, center=<-3.4696 -1.3695 0>}]
 spine_01 [{shapeType=Capsule, length=16.7187, radius=10.6638, center=<-2.6215 -1.1582 0>}]
 spine_02 [{shapeType=Capsule, length=15.0937, radius=10.695, center=<-2.6215 0.4993 0.0001>}]
 spine_03 [{shapeType=Capsule, length=13.6562, radius=10.5075}]
 spine_04 [{shapeType=Capsule, length=22.6562, radius=9.8825, center=<2.1467 -1.2748 0.0002>}]
 spine_05 [{shapeType=Capsule, length=22.0937, radius=12.945}]
 thigh_l []
 thigh_r []
 upperarm_l []
 upperarm_r []
}
constraints [
 {
  bone1Name=spine_02
  bone2Name=spine_01
  swing1LimitDegrees=45
  swing1Motion=Free
  swing2LimitDegrees=45
  swing2Motion=Free
  twistLimitDegrees=45
  twistMotion=Free
 }
 {
  bone1Name=spine_03
  bone2Name=spine_02
  swing1LimitDegrees=45
  swing1Motion=Free
  swing2LimitDegrees=45
  swing2Motion=Free
  twistLimitDegrees=45
  twistMotion=Free
 }
 {
  bone1Name=spine_04
  bone2Name=spine_03
  swing1LimitDegrees=45
  swing1Motion=Free
  swing2LimitDegrees=45
  swing2Motion=Free
  twistLimitDegrees=45
  twistMotion=Free
 }
 {
  bone1Name=spine_01
  bone2Name=pelvis
  swing1LimitDegrees=45
  swing1Motion=Free
  swing2LimitDegrees=45
  swing2Motion=Free
  twistLimitDegrees=45
  twistMotion=Free
 }
 {
  bone1Name=spine_05
  bone2Name=spine_04
  swing1LimitDegrees=45
  swing1Motion=Free
  swing2LimitDegrees=45
  swing2Motion=Free
  twistLimitDegrees=45
  twistMotion=Free
 }
 {
  bone1Name=upperarm_l
  bone2Name=spine_05
  swing1LimitDegrees=45
  swing1Motion=Free
  swing2LimitDegrees=45
  swing2Motion=Free
  twistLimitDegrees=45
  twistMotion=Free
 }
 {
  bone1Name=lowerarm_l
  bone2Name=upperarm_l
  swing1LimitDegrees=45
  swing1Motion=Free
  swing2LimitDegrees=45
  swing2Motion=Free
  twistLimitDegrees=45
  twistMotion=Free
 }
 {
  bone1Name=upperarm_r
  bone2Name=spine_05
  swing1LimitDegrees=45
  swing1Motion=Free
  swing2LimitDegrees=45
  swing2Motion=Free
  twistLimitDegrees=45
  twistMotion=Free
 }
 {
  bone1Name=lowerarm_r
  bone2Name=upperarm_r
  swing1LimitDegrees=45
  swing1Motion=Free
  swing2LimitDegrees=45
  swing2Motion=Free
  twistLimitDegrees=45
  twistMotion=Free
 }
 {
  bone1Name=thigh_l
  bone2Name=pelvis
  swing1LimitDegrees=45
  swing1Motion=Free
  swing2LimitDegrees=45
  swing2Motion=Free
  twistLimitDegrees=45
  twistMotion=Free
 }
 {
  bone1Name=thigh_r
  bone2Name=pelvis
  swing1LimitDegrees=45
  swing1Motion=Free
  swing2LimitDegrees=45
  swing2Motion=Free
  twistLimitDegrees=45
  twistMotion=Free
 }
 {
  bone1Name=calf_l
  bone2Name=thigh_l
  swing1LimitDegrees=45
  swing1Motion=Free
  swing2LimitDegrees=45
  swing2Motion=Free
  twistLimitDegrees=45
  twistMotion=Free
 }
 {
  bone1Name=calf_r
  bone2Name=thigh_r
  swing1LimitDegrees=45
  swing1Motion=Free
  swing2LimitDegrees=45
  swing2Motion=Free
  twistLimitDegrees=45
  twistMotion=Free
 }
 {
  bone1Name=dyn_cape_mid_l_base
  bone2Name=spine_05
  swing1LimitDegrees=45
  swing1Motion=Free
  swing2LimitDegrees=45
  swing2Motion=Free
  twistLimitDegrees=45
  twistMotion=Free
 }
 {bone1Name=dyn_cape_mid_01_l, bone2Name=dyn_cape_mid_l_base, swing1LimitDegrees=25, swing2LimitDegrees=15, twistLimitDegrees=15}
 {bone1Name=dyn_cape_mid_02_l, bone2Name=dyn_cape_mid_01_l, swing1LimitDegrees=25, swing2LimitDegrees=15, twistLimitDegrees=15}
 {bone1Name=dyn_cape_mid_03_l, bone2Name=dyn_cape_mid_02_l, swing1LimitDegrees=15, swing2LimitDegrees=15, twistLimitDegrees=15}
 {bone1Name=dyn_cape_mid_04_l, bone2Name=dyn_cape_mid_03_l, swing1LimitDegrees=15, swing2LimitDegrees=15, twistLimitDegrees=15}
 {
  bone1Name=dyn_cape_l_base
  bone2Name=upperarm_l
  swing1LimitDegrees=45
  swing1Motion=Free
  swing2LimitDegrees=45
  swing2Motion=Free
  twistLimitDegrees=45
  twistMotion=Free
 }
 {bone1Name=dyn_cape_01_l, bone2Name=dyn_cape_l_base, swing1LimitDegrees=25, swing2LimitDegrees=30, twistLimitDegrees=15}
 {bone1Name=dyn_cape_02_l, bone2Name=dyn_cape_01_l, swing1LimitDegrees=15, swing2LimitDegrees=15, twistLimitDegrees=15}
 {bone1Name=dyn_cape_03_l, bone2Name=dyn_cape_02_l, swing1LimitDegrees=15, swing2LimitDegrees=15, twistLimitDegrees=15}
 {bone1Name=dyn_cape_04_l, bone2Name=dyn_cape_03_l, swing1LimitDegrees=15, swing2LimitDegrees=15, twistLimitDegrees=15}
 {
  bone1Name=dyn_cape_mid_r_base
  bone2Name=spine_05
  swing1LimitDegrees=45
  swing1Motion=Free
  swing2LimitDegrees=45
  swing2Motion=Free
  twistLimitDegrees=45
  twistMotion=Free
 }
 {bone1Name=dyn_cape_mid_01_r, bone2Name=dyn_cape_mid_r_base, swing1LimitDegrees=15, swing2LimitDegrees=15, twistLimitDegrees=15}
 {bone1Name=dyn_cape_mid_02_r, bone2Name=dyn_cape_mid_01_r, swing1LimitDegrees=15, swing2LimitDegrees=15, twistLimitDegrees=15}
 {bone1Name=dyn_cape_mid_03_r, bone2Name=dyn_cape_mid_02_r, swing1LimitDegrees=15, swing2LimitDegrees=15, twistLimitDegrees=15}
 {bone1Name=dyn_cape_mid_04_r, bone2Name=dyn_cape_mid_03_r, swing1LimitDegrees=15, swing2LimitDegrees=15, twistLimitDegrees=15}
 {
  bone1Name=dyn_cape_r_base
  bone2Name=upperarm_r
  swing1LimitDegrees=45
  swing1Motion=Free
  swing2LimitDegrees=45
  swing2Motion=Free
  twistLimitDegrees=45
  twistMotion=Free
 }
 {bone1Name=dyn_cape_01_r, bone2Name=dyn_cape_r_base, swing1LimitDegrees=25, swing2LimitDegrees=30, twistLimitDegrees=15}
 {bone1Name=dyn_cape_02_r, bone2Name=dyn_cape_01_r, swing1LimitDegrees=25, swing2LimitDegrees=15, twistLimitDegrees=15}
 {bone1Name=dyn_cape_03_r, bone2Name=dyn_cape_02_r, swing1LimitDegrees=25, swing2LimitDegrees=15, twistLimitDegrees=15}
 {bone1Name=dyn_cape_04_r, bone2Name=dyn_cape_03_r, swing1LimitDegrees=25, swing2LimitDegrees=15, twistLimitDegrees=15}
 {
  bone1Name=dyn_cape_back_l_base
  bone2Name=spine_05
  swing1LimitDegrees=45
  swing1Motion=Free
  swing2LimitDegrees=45
  swing2Motion=Free
  twistLimitDegrees=45
  twistMotion=Free
 }
 {bone1Name=dyn_cape_back_01_l, bone2Name=dyn_cape_back_l_base, swing1LimitDegrees=15, swing2LimitDegrees=15, twistLimitDegrees=15}
 {bone1Name=dyn_cape_back_02_l, bone2Name=dyn_cape_back_01_l, swing1LimitDegrees=15, swing2LimitDegrees=15, twistLimitDegrees=15}
 {bone1Name=dyn_cape_back_03_l, bone2Name=dyn_cape_back_02_l, swing1LimitDegrees=15, swing2LimitDegrees=15, twistLimitDegrees=15}
 {bone1Name=dyn_cape_back_04_l, bone2Name=dyn_cape_back_03_l, swing1LimitDegrees=15, swing2LimitDegrees=15, twistLimitDegrees=15}
 {
  bone1Name=dyn_cape_back_r_base
  bone2Name=spine_05
  swing1LimitDegrees=45
  swing1Motion=Free
  swing2LimitDegrees=45
  swing2Motion=Free
  twistLimitDegrees=45
  twistMotion=Free
 }
 {bone1Name=dyn_cape_back_01_r, bone2Name=dyn_cape_back_r_base, swing1LimitDegrees=15, swing2LimitDegrees=15, twistLimitDegrees=15}
 {bone1Name=dyn_cape_back_02_r, bone2Name=dyn_cape_back_01_r, swing1LimitDegrees=15, swing2LimitDegrees=15, twistLimitDegrees=15}
 {bone1Name=dyn_cape_back_03_r, bone2Name=dyn_cape_back_02_r, swing1LimitDegrees=15, swing2LimitDegrees=15, twistLimitDegrees=15}
 {bone1Name=dyn_cape_back_04_r, bone2Name=dyn_cape_back_03_r, swing1LimitDegrees=15, swing2LimitDegrees=15, twistLimitDegrees=15}
]
