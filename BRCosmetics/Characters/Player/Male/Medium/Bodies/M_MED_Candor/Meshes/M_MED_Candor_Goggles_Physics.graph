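# Physics asset for the goggles on the male medium Candor cosmetic body, set up for secondary motion of the goggle bones.
# A four-bone capsule chain hangs from a small base body through angular constraints, with twist locked at the first joint.
# The third link adds a larger angled capsule for extra volume.
# Wide spine capsules provide torso shapes, linked to each other and the goggle base by free-rotation constraints.
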PhysicsAsset /BRCosmetics/Characters/Player/Male/Medium/Bodies/M_MED_Candor/Meshes/M_MED_Candor_Goggles_Physics
bodies {
 dyn_goggles_01 [{shapeType=Capsule, length=2, radius=2, shapeName=dyn_goggles_01_capsule, center=<2 0 0>, rotation=<rot 90 0 0>}]
 dyn_goggles_02 [{shapeType=Capsule, length=2, radius=2, shapeName=dyn_goggles_02_capsule, center=<2 0 0>, rotation=<rot 90 0 0>}]
 dyn_goggles_03 [
  {shapeType=Capsule, length=2, radius=2, shapeName=dyn_goggles_03_capsule, center=<2 0 0>, rotation=<rot 90 0 0>}
  {shapeType=Capsule, length=7.75, radius=4, shapeName=dyn_goggles_03_capsule, center=<7.7993 0 2.1094>, rotation=<rot -40.0001 0.0001 0.0001>}
 ]
 dyn_goggles_04 [{shapeType=Capsule, length=2, radius=2, shapeName=dyn_goggles_04_capsule, center=<2 0 0>, rotation=<rot 90 0 0>}]
 dyn_goggles_base [{shapeType=Capsule, length=0.505, radius=0.505, shapeName=dyn_goggles_base_capsule, rotation=<rot 0 -90 0>}]
 spine_01 [{shapeType=Capsule, length=2.9687, radius=12.4138, center=<-2.6215 -0.7156 0>}]
 spine_02 [{shapeType=Capsule, length=5.3437, radius=10.945, center=<-2.6215 0.4993 0.0001>}]
 spine_03 [{shapeType=Capsule, length=9.1562, radius=10.7575, center=<-1.953 0.1364 0.0001>}]
}
constraints [
 {
  bone1Name=spine_02
  bone2Name=spine_01
  swing1LimitDegrees=45
  swing1Motion=Free
  swing2LimitDegrees=45
  swing2Motion=Free
  twistLimitDegrees=45
  twistMotion=Free
 }
 {
  bone1Name=spine_03
  bone2Name=spine_02
  swing1LimitDegrees=45
  swing1Motion=Free
  swing2LimitDegrees=45
  swing2Motion=Free
  twistLimitDegrees=45
  twistMotion=Free
 }
 {bone1Name=dyn_goggles_01, bone2Name=dyn_goggles_base, swing1LimitDegrees=30, swing2LimitDegrees=10, twistLimitDegrees=15, twistMotion=Locked}
 {bone1Name=dyn_goggles_02, bone2Name=dyn_goggles_01, swing1LimitDegrees=30, swing2LimitDegrees=30, twistLimitDegrees=15}
 {bone1Name=dyn_goggles_03, bone2Name=dyn_goggles_02, swing1LimitDegrees=30, swing2LimitDegrees=30, twistLimitDegrees=15}
 {bone1Name=dyn_goggles_04, bone2Name=dyn_goggles_03, swing1LimitDegrees=30, swing2LimitDegrees=30, twistLimitDegrees=15}
 {
  bone1Name=dyn_goggles_base
  bone2Name=spine_03
  swing1LimitDegrees=45
  swing1Motion=Free
  swing2LimitDegrees=45
  swing2Motion=Free
  twistLimitDegrees=45
  twistMotion=Free
 }
]
