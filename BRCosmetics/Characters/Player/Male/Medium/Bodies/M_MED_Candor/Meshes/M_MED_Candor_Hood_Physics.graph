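# Physics asset for the hood of the M_MED_Candor male player cosmetic body, covering its dangling laces and strap.
# Small capsules form three-link chains for the left lace, right lace and strap, each constrained to spine_05, with tighter swing limits on the strap.
# Large capsules on spine_04 and spine_05 represent the upper torso, and the constraint between them leaves swing and twist free.
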PhysicsAsset /BRCosmetics/Characters/Player/Male/Medium/Bodies/M_MED_Candor/Meshes/M_MED_Candor_Hood_Physics
bodies {
 dyn_hood_lace_01_l [{shapeType=Capsule, length=1, radius=1, shapeName=dyn_hood_lace_01_l_capsule, center=<1 0 0>, rotation=<rot 90 0 0>}]
 dyn_hood_lace_01_r [{shapeType=Capsule, length=1.5, radius=2, shapeName=dyn_hood_lace_01_r_capsule, center=<2.7052 0 0>, rotation=<rot 90 0 0>}]
 dyn_hood_lace_02_l [{shapeType=Capsule, length=1, radius=1, shapeName=dyn_hood_lace_02_l_capsule, center=<1.7629 0 0>, rotation=<rot 90 0 0>}]
 dyn_hood_lace_02_r [{shapeType=Capsule, length=2.25, radius=1.25, shapeName=dyn_hood_lace_02_r_capsule, center=<3.9265 0 0>, rotation=<rot 90 0 0>}]
 dyn_hood_lace_03_l [{shapeType=Capsule, length=1, radius=1, shapeName=dyn_hood_lace_03_l_capsule, center=<0 0 -1.753>, rotation=<rot 90 0 0>}]
 dyn_hood_lace_03_r [{shapeType=Capsule, length=2.5, radius=1, shapeName=dyn_hood_lace_03_r_capsule, center=<3.6103 0 0>, rotation=<rot 90 0 0>}]
 dyn_hood_strap_01 [{shapeType=Capsule, length=2, radius=2, shapeName=dyn_hood_strap_01_capsule, center=<1 0 0>}]
 dyn_hood_strap_02 [{shapeType=Capsule, length=2, radius=2, shapeName=dyn_hood_strap_02_capsule, center=<1 0 0>}]
 dyn_hood_strap_03 [{shapeType=Capsule, length=2, radius=2, shapeName=dyn_hood_strap_03_capsule, center=<1 0 0>}]
 spine_04 [{shapeType=Capsule, length=5.1562, radius=15.8825, center=<2.1467 -1.2748 0.0002>}]
 spine_05 [{shapeType=Capsule, length=12.8437, radius=10.945, center=<1.5384 -1.7167 0>}]
}
constraints [
 {
  bone1Name=spine_05
  bone2Name=spine_04
  swing1LimitDegrees=45
  swing1Motion=Free
  swing2LimitDegrees=45
  swing2Motion=Free
  twistLimitDegrees=45
  twistMotion=Free
 }
 {bone1Name=dyn_hood_lace_01_l, bone2Name=spine_05, swing1LimitDegrees=30, swing2LimitDegrees=30, twistLimitDegrees=15}
 {bone1Name=dyn_hood_lace_02_l, bone2Name=dyn_hood_lace_01_l, swing1LimitDegrees=30, swing2LimitDegrees=30, twistLimitDegrees=15}
 {bone1Name=dyn_hood_lace_03_l, bone2Name=dyn_hood_lace_02_l, swing1LimitDegrees=30, swing2LimitDegrees=30, twistLimitDegrees=15}
 {bone1Name=dyn_hood_lace_01_r, bone2Name=spine_05, swing1LimitDegrees=30, swing2LimitDegrees=30, twistLimitDegrees=15}
 {bone1Name=dyn_hood_lace_02_r, bone2Name=dyn_hood_lace_01_r, swing1LimitDegrees=30, swing2LimitDegrees=30, twistLimitDegrees=15}
 {bone1Name=dyn_hood_lace_03_r, bone2Name=dyn_hood_lace_02_r, swing1LimitDegrees=30, swing2LimitDegrees=30, twistLimitDegrees=15}
 {bone1Name=dyn_hood_strap_01, bone2Name=spine_05, swing1LimitDegrees=15, swing2LimitDegrees=15, twistLimitDegrees=15}
 {bone1Name=dyn_hood_strap_02, bone2Name=dyn_hood_strap_01, swing1LimitDegrees=15, swing2LimitDegrees=15, twistLimitDegrees=15}
 {bone1Name=dyn_hood_strap_03, bone2Name=dyn_hood_strap_02, swing1LimitDegrees=15, swing2LimitDegrees=15, twistLimitDegrees=15}
]
