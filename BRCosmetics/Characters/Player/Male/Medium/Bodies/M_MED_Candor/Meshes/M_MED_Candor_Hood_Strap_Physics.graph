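# Physics asset for the hood strap and zipper collar on the M_MED_Candor male medium player body cosmetic, giving their dyn bones capsule bodies.
# Two three-bone chains, one for the strap and one for the collar, each link bone to bone and anchor to spine_05.
# Chain joints carry narrow swing and twist limits, with tighter twist on the collar.
# Larger spine_04 and spine_05 capsules represent the torso, joined by an angularly free constraint.
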PhysicsAsset /BRCosmetics/Characters/Player/Male/Medium/Bodies/M_MED_Candor/Meshes/M_MED_Candor_Hood_Strap_Physics
bodies {
 dyn_hood_strap_01 [{shapeType=Capsule, length=2, radius=2, shapeName=dyn_hood_strap_01_capsule, center=<1 0 0>}]
 dyn_hood_strap_02 [{shapeType=Capsule, length=2, radius=2, shapeName=dyn_hood_strap_02_capsule, center=<1 0 0>}]
 dyn_hood_strap_03 [{shapeType=Capsule, length=2, radius=2, shapeName=dyn_hood_strap_03_capsule, center=<1 0 0>}]
 dyn_zipper_collar_01 [{shapeType=Capsule, length=2, radius=1.5, shapeName=dyn_zipper_collar_01_capsule, center=<2 0 0>, rotation=<rot 90 0 0>}]
 dyn_zipper_collar_02 [{shapeType=Capsule, length=2, radius=1.5, shapeName=dyn_zipper_collar_02_capsule, center=<2 0 0>, rotation=<rot 90 0 0>}]
 dyn_zipper_collar_03 [{shapeType=Capsule, length=2, radius=1.5, shapeName=dyn_zipper_collar_03_capsule, center=<2 0 0>, rotation=<rot 90 0 0>}]
 spine_04 [{shapeType=Capsule, length=5.1562, radius=15.8825, center=<2.1467 -1.2748 0.0002>}]
 spine_05 [{shapeType=Capsule, length=12.8437, radius=10.945, center=<1.5384 -1.7167 0>}]
}
constraints [
 {
  bone1Name=spine_05
  bone2Name=spine_04
  swing1LimitDegrees=45
  swing1Motion=Free
  swing2LimitDegrees=45
  swing2Motion=Free
  twistLimitDegrees=45
  twistMotion=Free
 }
 {bone1Name=dyn_hood_strap_01, bone2Name=spine_05, swing1LimitDegrees=15, swing2LimitDegrees=15, twistLimitDegrees=15}
 {bone1Name=dyn_hood_strap_02, bone2Name=dyn_hood_strap_01, swing1LimitDegrees=15, swing2LimitDegrees=15, twistLimitDegrees=15}
 {bone1Name=dyn_hood_strap_03, bone2Name=dyn_hood_strap_02, swing1LimitDegrees=15, swing2LimitDegrees=15, twistLimitDegrees=15}
 {bone1Name=dyn_zipper_collar_01, bone2Name=spine_05, swing1LimitDegrees=15, swing2LimitDegrees=15, twistLimitDegrees=5}
 {bone1Name=dyn_zipper_collar_02, bone2Name=dyn_zipper_collar_01, swing1LimitDegrees=15, swing2LimitDegrees=15, twistLimitDegrees=5}
 {bone1Name=dyn_zipper_collar_03, bone2Name=dyn_zipper_collar_02, swing1LimitDegrees=15, swing2LimitDegrees=15, twistLimitDegrees=5}
]
