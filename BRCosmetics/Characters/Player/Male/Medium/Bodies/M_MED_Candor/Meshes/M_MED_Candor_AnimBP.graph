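# Post-process-style AnimBlueprint for the Candor body cosmetic, copying the parent mesh pose and adding deformation and secondary motion to cape, hood, laces and goggles.
# Control Rigs, constraints, AnimDynamics and goggle RigidBody build a cached base pose.
# Cape RigidBody setups switch by frontend, crouch, motorcycle and backbling state.
# With rigid bodies disabled, additive bone offsets replace simulation, partly driven by head and arm rotations computed in the EventGraph alongside gravity override.
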AnimBlueprint /BRCosmetics/Characters/Player/Male/Medium/Bodies/M_MED_Candor/Meshes/M_MED_Candor_AnimBP
target_skeleton=&/M_MED_Candor_Skeleton
graphs [
 AnimGraph {nodes [
   Output Pose(Result=Blend Poses by bool(BlendPose_0=Use cached pose 'RBAN'(), BlendPose_1=Use cached pose 'RBAN Disabled'(), bActiveValue=Get bRigidBodyEnabled()) {blendType=Linear}) {layerGroup=None}
   n1 = Control Rig(Source=Copy Pose From Mesh() {bCopyCurves=true, bUseAttachedParent=true}) {
    alphaBoolBlend {blendOption=Linear}
    controlRigAssetReference {blueprintRigClass=&/M_MED_Candor_DeformRig_ControlRig.M_MED_Candor_DeformRig_ControlRig_C}
   }
   n2 = Control Rig!(Source=@n1, Alpha, isAiming:bool=Get bIsTargeting(), isCrouching:bool=Get bIsCrouching()) -> (Pose) {
    alphaBoolBlend {blendOption=Linear}
    controlRigAssetReference {blueprintRigClass=&/M_MED_Candor_CtrlRig.M_MED_Candor_CtrlRig_C}
   }
   n3 = "Constraint - dyn_cape_mid_r_base (PP)"!(ComponentPose=Local To Component(LocalPose=@n2), Alpha=1, ConstraintWeights_0=0.4, ConstraintWeights_1:float=0.6) -> (Pose) {
    alphaBoolBlend {blendOption=Linear}
    boneToModify {boneName=dyn_cape_mid_r_base}
    constraintSetup [{targetBone {boneName=upperarm_r}}, {targetBone {boneName=clavicle_r}}]
    constraintWeights [1,1]
   }
   n4 = "Constraint - dyn_cape_mid_l_base (PP)"!(ComponentPose=@n3, Alpha=1, ConstraintWeights_0=0.4, ConstraintWeights_1:float=0.6) -> (Pose) {
    alphaBoolBlend {blendOption=Linear}
    boneToModify {boneName=dyn_cape_mid_l_base}
    constraintSetup [{targetBone {boneName=upperarm_l}}, {targetBone {boneName=clavicle_l}}]
    constraintWeights [1,1]
   }
   n5 = "Constraint - dyn_hood_lace_01_l (PP)"!(ComponentPose=@n4, Alpha=1, ConstraintWeights_0=0.9, ConstraintWeights_1:float=0.1) -> (Pose) {
    alphaBoolBlend {blendOption=Linear}
    boneToModify {boneName=dyn_hood_lace_01_l}
    constraintSetup [{targetBone {boneName=spine_05}}, {targetBone {boneName=clavicle_l}}]
    constraintWeights [1,1]
   }
   n6 = "Constraint - dyn_goggles_base (PPPPP)"!(
    ComponentPose=@n5
    Alpha=1
    ConstraintWeights_0=0.15
    ConstraintWeights_1:float=0.3
    ConstraintWeights_2:float=0.2
    ConstraintWeights_3:float=0.125
    ConstraintWeights_4:float=0.225
   ) -> (Pose) {
    alphaBoolBlend {blendOption=Linear}
    boneToModify {boneName=dyn_goggles_base}
    constraintSetup [
     {perAxis {bX=false, bY=false}, targetBone {boneName=spine_02}}
     {perAxis {bX=false, bY=false}, targetBone {boneName=spine_03}}
     {perAxis {bX=false, bY=false}, targetBone {boneName=spine_04}}
     {targetBone {boneName=clavicle_l}}
     {targetBone {boneName=deform_pec_l}}
    ]
    constraintWeights [1,1,1,1,1]
   }
   n7 = "Transform (Modify) Bone - Bone: dyn_hood"(ComponentPose=@n6, Alpha=Get headAlpha() -> (headAlpha:double), Translation="0,0,0", Rotation="15,-20,0.000", Scale="1,1,1") {
    rotationMode=BMM_Additive
    rotationSpace=BCS_BoneSpace
    alphaBoolBlend {blendOption=Linear}
    alphaScaleBiasClamp {bInterpResult=true, interpSpeedIncreasing=20}
    boneToModify {boneName=dyn_hood}
    rotation {pitch=15, yaw=-20}
   }
   n8 = Get GravityOverride()
   n9 = Anim Dynamics - Bone: dyn_hood_lace_01_l!(ComponentPose=@n7, Alpha=1, GravityOverride:Vector=@n8, ExternalForce="0,0,0") -> (Pose) {
    angularDampingOverride=0.995
    angularSpringConstant=2500
    bOverrideAngularDamping=true
    bUseGravityOverride=true
    simulationSpace=BoneRelative
    alphaBoolBlend {blendOption=Linear}
    boundBone {boneName=dyn_hood_lace_01_l}
    componentAppliedLinearAccClamp=<1200 1200 1200>
    componentLinearAccScale=<0.15 0.15 0.15>
    componentLinearVelScale=<0.15 0.15 0.15>
    gravityOverride {z=-980}
    physicsBodyDefinitions [{
      boundBone {boneName=dyn_hood_lace_01_l}
      boxExtents=<5 3 5>
      constraintSetup {angularLimitsMax=<5 10 15>, angularLimitsMin=<-5 -5 10>, angularTarget=<0 0 0>}
      localJointOffset=<2 0 0>
     }]
    relativeSpaceBone {boneName=spine_05}
    sphericalLimits [
     {limitRadius=13, drivingBone {boneName=spine_05}, sphereLocalOffset=<1.5 -1.7 0>}
     {limitRadius=11, drivingBone {boneName=spine_05}, sphereLocalOffset=<2 0 10>}
     {limitRadius=11, drivingBone {boneName=spine_05}, sphereLocalOffset=<2 0 -10>}
    ]
   }
   n10 = Anim Dynamics - Bone: dyn_hood_lace_02_l!(ComponentPose=@n9, Alpha=1, GravityOverride:Vector=@n8, ExternalForce="0,0,0") -> (Pose) {
    angularDampingOverride=0.995
    angularSpringConstant=2500
    bOverrideAngularDamping=true
    bUseGravityOverride=true
    simulationSpace=BoneRelative
    alphaBoolBlend {blendOption=Linear}
    boundBone {boneName=dyn_hood_lace_02_l}
    componentAppliedLinearAccClamp=<1200 1200 1200>
    componentLinearAccScale=<0.15 0.15 0.15>
    componentLinearVelScale=<0.15 0.15 0.15>
    gravityOverride {z=-980}
    physicsBodyDefinitions [{
      boundBone {boneName=dyn_hood_lace_02_l}
      boxExtents=<5 3 5>
      constraintSetup {angularLimitsMax=<5 10 15>, angularLimitsMin=<-5 -5 10>, angularTarget=<0 0 0>}
      localJointOffset=<2 0 0>
     }]
    relativeSpaceBone {boneName=spine_05}
    sphericalLimits [
     {limitRadius=13, drivingBone {boneName=spine_05}, sphereLocalOffset=<1.5 -1.7 0>}
     {limitRadius=11, drivingBone {boneName=spine_05}, sphereLocalOffset=<2 0 10>}
     {limitRadius=11, drivingBone {boneName=spine_05}, sphereLocalOffset=<2 0 -10>}
    ]
   }
   n11 = Anim Dynamics - Bone: dyn_hood_lace_01_r!(ComponentPose=@n10, Alpha=1, GravityOverride:Vector=@n8, ExternalForce="0,0,0") -> (Pose) {
    angularDampingOverride=0.995
    angularSpringConstant=2500
    bOverrideAngularDamping=true
    bUseGravityOverride=true
    simulationSpace=BoneRelative
    alphaBoolBlend {blendOption=Linear}
    boundBone {boneName=dyn_hood_lace_01_r}
    componentAppliedLinearAccClamp=<1200 1200 1200>
    componentLinearAccScale=<0.15 0.15 0.15>
    componentLinearVelScale=<0.15 0.15 0.15>
    gravityOverride {z=-980}
    physicsBodyDefinitions [{
      boundBone {boneName=dyn_hood_lace_01_r}
      boxExtents=<6 4 6>
      constraintSetup {angularLimitsMax=<5 10 10>, angularLimitsMin=<-5 -10 0>, angularTarget=<0 0 0>}
      localJointOffset=<2 0 0>
     }]
    relativeSpaceBone {boneName=spine_05}
    sphericalLimits [
     {limitRadius=13, drivingBone {boneName=spine_05}, sphereLocalOffset=<1.5 -1.7 0>}
     {limitRadius=10, drivingBone {boneName=spine_05}, sphereLocalOffset=<2 0 10>}
     {limitRadius=10, drivingBone {boneName=spine_05}, sphereLocalOffset=<2 0 -10>}
    ]
   }
   n12 = Anim Dynamics - Bone: dyn_hood_lace_02_r!(ComponentPose=@n11, Alpha=1, GravityOverride:Vector=@n8, ExternalForce="0,0,0") -> (Pose) {
    angularDampingOverride=0.995
    angularSpringConstant=2500
    bOverrideAngularDamping=true
    bUseGravityOverride=true
    simulationSpace=BoneRelative
    alphaBoolBlend {blendOption=Linear}
    boundBone {boneName=dyn_hood_lace_02_r}
    componentAppliedLinearAccClamp=<1200 1200 1200>
    componentLinearAccScale=<0.15 0.15 0.15>
    componentLinearVelScale=<0.15 0.15 0.15>
    gravityOverride {z=-980}
    physicsBodyDefinitions [{
      boundBone {boneName=dyn_hood_lace_02_r}
      boxExtents=<6 4 6>
      constraintSetup {angularLimitsMax=<5 10 10>, angularLimitsMin=<-5 -10 0>, angularTarget=<0 0 0>}
      localJointOffset=<2 0 0>
     }]
    relativeSpaceBone {boneName=spine_05}
    sphericalLimits [
     {limitRadius=13, drivingBone {boneName=spine_05}, sphereLocalOffset=<1.5 -1.7 0>}
     {limitRadius=10, drivingBone {boneName=spine_05}, sphereLocalOffset=<2 0 10>}
     {limitRadius=10, drivingBone {boneName=spine_05}, sphereLocalOffset=<2 0 -10>}
    ]
   }
   n13 = Anim Dynamics - Bone: dyn_hood_strap_01!(ComponentPose=@n12, Alpha=1, GravityOverride:Vector=@n8, ExternalForce="0,0,0") -> (Pose) {
    angularDampingOverride=0.9999
    angularSpringConstant=2500
    bOverrideAngularDamping=true
    bUseGravityOverride=true
    simulationSpace=BoneRelative
    alphaBoolBlend {blendOption=Linear}
    boundBone {boneName=dyn_hood_strap_01}
    componentAppliedLinearAccClamp=<1000 1000 1000>
    componentLinearAccScale=<0.15 0.15 0.15>
    componentLinearVelScale=<0.15 0.15 0.15>
    gravityOverride {z=-980}
    physicsBodyDefinitions [{
      boundBone {boneName=dyn_hood_strap_01}
      boxExtents=<7 3.5 7>
      constraintSetup {angularLimitsMax=<5 5 10>, angularLimitsMin=<-5 -5 0>, angularTarget=<0 0 0>}
      localJointOffset=<3 0 0>
     }]
    relativeSpaceBone {boneName=spine_05}
    sphericalLimits [
     {limitRadius=12, drivingBone {boneName=spine_05}, sphereLocalOffset=<1.5 -1.7 0>}
     {limitRadius=9, drivingBone {boneName=spine_05}, sphereLocalOffset=<2 0 10>}
     {limitRadius=9, drivingBone {boneName=spine_05}, sphereLocalOffset=<2 0 -10>}
    ]
   }
   n14 = Anim Dynamics - Bone: dyn_hood_strap_02!(ComponentPose=@n13, Alpha=1, GravityOverride:Vector=@n8, ExternalForce="0,0,0") -> (Pose) {
    angularDampingOverride=0.9999
    angularSpringConstant=2500
    bOverrideAngularDamping=true
    bUseGravityOverride=true
    simulationSpace=BoneRelative
    alphaBoolBlend {blendOption=Linear}
    boundBone {boneName=dyn_hood_strap_02}
    componentAppliedLinearAccClamp=<1000 1000 1000>
    componentLinearAccScale=<0.15 0.15 0.15>
    componentLinearVelScale=<0.15 0.15 0.15>
    gravityOverride {z=-980}
    physicsBodyDefinitions [{
      boundBone {boneName=dyn_hood_strap_02}
      boxExtents=<7 3.5 7>
      constraintSetup {angularLimitsMax=<5 5 10>, angularLimitsMin=<-5 -5 0>, angularTarget=<0 0 0>}
      localJointOffset=<3 0 0>
     }]
    relativeSpaceBone {boneName=spine_05}
    sphericalLimits [
     {limitRadius=12, drivingBone {boneName=spine_05}, sphereLocalOffset=<1.5 -1.7 0>}
     {limitRadius=9, drivingBone {boneName=spine_05}, sphereLocalOffset=<2 0 10>}
     {limitRadius=9, drivingBone {boneName=spine_05}, sphereLocalOffset=<2 0 -10>}
    ]
   }
   n15 = RigidBody!(ComponentPose=@n14, Alpha, OverrideWorldGravity:Vector=Get GravityOverride(), ExternalForce) -> (Pose) {
    bOverrideWorldGravity=true
    overridePhysicsAsset=&/M_MED_Candor_Goggles_Physics
    simulationSpace=BaseBoneSpace
    alphaBoolBlend {blendOption=Linear}
    baseBoneRef {boneName=pelvis}
    componentAppliedLinearAccClamp=<2000 2000 2000>
    componentLinearAccScale=<0.5 0.5 0.5>
    componentLinearVelScale=<0.5 0.5 0.5>
    overrideWorldGravity {z=-980}
   }
   Save cached pose 'Base Pose'(Pose=Component To Local(ComponentPose=@n15)) {cacheName=Base Pose}
   n16 = Get isWearingBackbling()
   n17 = RigidBody!(
    ComponentPose=Local To Component(LocalPose=Use cached pose 'Base Pose'())
    bAlphaBoolEnabled:bool=NOT Boolean(A=@n16)
    OverrideWorldGravity="0,0,-980"
    ExternalForce
   ) -> (Pose) {
    alphaInputType=Bool
    bOverrideWorldGravity=true
    overridePhysicsAsset=&/M_MED_Candor_Cape_Physics
    simulationSpace=BaseBoneSpace
    alphaBoolBlend {blendOption=Linear}
    baseBoneRef {boneName=spine_05}
    componentAppliedLinearAccClamp=<2000 2000 2000>
    componentLinearAccScale=<0.3 0.3 0.3>
    componentLinearVelScale=<0.3 0.3 0.3>
    overrideWorldGravity {z=-980}
   }
   n18 = RigidBody(ComponentPose=@n17, bAlphaBoolEnabled=@n16, OverrideWorldGravity="0,0,-980") {
    alphaInputType=Bool
    bOverrideWorldGravity=true
    overridePhysicsAsset=&/M_MED_Candor_Cape_Backpack_Physics
    simulationSpace=BaseBoneSpace
    alphaBoolBlend {blendOption=Linear}
    baseBoneRef {boneName=spine_05}
    componentAppliedLinearAccClamp=<1500 1500 1500>
    componentLinearAccScale=<0.2 0.2 0.25>
    componentLinearVelScale=<0.2 0.2 0.25>
    overrideWorldGravity {z=-980}
   }
   n19 = "Transform (Modify) Bone - Bone: dyn_cape_back_l_base"(ComponentPose=Local To Component(LocalPose=Use cached pose 'Base Pose'()), Alpha=1, Translation="0,0,0", Rotation="0,-20,0", Scale="1,1,1") {
    rotationMode=BMM_Additive
    rotationSpace=BCS_BoneSpace
    alphaBoolBlend {blendOption=Linear}
    alphaScaleBiasClamp {bInterpResult=true, interpSpeedIncreasing=20}
    boneToModify {boneName=dyn_cape_back_l_base}
    rotation {yaw=-20}
   }
   n20 = "Transform (Modify) Bone - Bone: dyn_cape_mid_l_base"(ComponentPose=@n19, Alpha=1, Translation="0,0,0", Rotation="0,-20,0", Scale="1,1,1") {
    rotationMode=BMM_Additive
    rotationSpace=BCS_BoneSpace
    alphaBoolBlend {blendOption=Linear}
    alphaScaleBiasClamp {bInterpResult=true, interpSpeedIncreasing=20}
    boneToModify {boneName=dyn_cape_mid_l_base}
    rotation {yaw=-20}
   }
   n21 = "Transform (Modify) Bone - Bone: dyn_cape_l_base"(ComponentPose=@n20, Alpha=1, Translation="0,0,0", Rotation="0,-20,0", Scale="1,1,1") {
    rotationMode=BMM_Additive
    rotationSpace=BCS_BoneSpace
    alphaBoolBlend {blendOption=Linear}
    alphaScaleBiasClamp {bInterpResult=true, interpSpeedIncreasing=20}
    boneToModify {boneName=dyn_cape_l_base}
    rotation {yaw=-20}
   }
   n22 = "Transform (Modify) Bone - Bone: dyn_cape_back_r_base"(ComponentPose=@n21, Alpha=1, Translation="0,0,0", Rotation="0,-20,0", Scale="1,1,1") {
    rotationMode=BMM_Additive
    rotationSpace=BCS_BoneSpace
    alphaBoolBlend {blendOption=Linear}
    alphaScaleBiasClamp {bInterpResult=true, interpSpeedIncreasing=20}
    boneToModify {boneName=dyn_cape_back_r_base}
    rotation {yaw=-20}
   }
   n23 = "Transform (Modify) Bone - Bone: dyn_cape_mid_r_base"(ComponentPose=@n22, Alpha=1, Translation="0,0,0", Rotation="0,-20,0", Scale="1,1,1") {
    rotationMode=BMM_Additive
    rotationSpace=BCS_BoneSpace
    alphaBoolBlend {blendOption=Linear}
    alphaScaleBiasClamp {bInterpResult=true, interpSpeedIncreasing=20}
    boneToModify {boneName=dyn_cape_mid_r_base}
    rotation {yaw=-20}
   }
   n24 = "Transform (Modify) Bone - Bone: dyn_cape_r_base"(ComponentPose=@n23, Alpha=1, Translation="0,0,0", Rotation="0,-20,0", Scale="1,1,1") {
    rotationMode=BMM_Additive
    rotationSpace=BCS_BoneSpace
    alphaBoolBlend {blendOption=Linear}
    alphaScaleBiasClamp {bInterpResult=true, interpSpeedIncreasing=20}
    boneToModify {boneName=dyn_cape_r_base}
    rotation {yaw=-20}
   }
   n25 = RigidBody(ComponentPose=@n24, OverrideWorldGravity="0,0,-980") {
    alphaInputType=Bool
    bOverrideWorldGravity=true
    overridePhysicsAsset=&/M_MED_Candor_Cape_Motorbike_Physics
    simulationSpace=BaseBoneSpace
    alphaBoolBlend {blendOption=Linear}
    baseBoneRef {boneName=spine_05}
    componentAppliedLinearAccClamp=<2000 2000 2000>
    componentLinearAccScale=<0.3 0.3 0.3>
    componentLinearVelScale=<0.3 0.3 0.3>
    overrideWorldGravity {z=-980}
   }
   n26 = Blend Poses by bool(
    BlendPose_0=Component To Local(ComponentPose=@n25)
    BlendPose_1=Component To Local(ComponentPose=@n18)
    BlendTime_0=0.25
    BlendTime_1=0.25
    bActiveValue=Get bIsMotorcycleDriver() -> (bIsMotorcycleDriver:bool)
   ) {blendType=Linear}
   n27 = RigidBody(ComponentPose=Local To Component(LocalPose=Use cached pose 'Base Pose'()), OverrideWorldGravity="0,0,-980") {
    alphaInputType=Bool
    bOverrideWorldGravity=true
    overridePhysicsAsset=&/M_MED_Candor_Cape_Crouch_Physics
    simulationSpace=BaseBoneSpace
    alphaBoolBlend {blendOption=Linear}
    baseBoneRef {boneName=spine_05}
    componentAppliedLinearAccClamp=<2000 2000 2000>
    componentLinearAccScale=<0.3 0.3 0.3>
    componentLinearVelScale=<0.3 0.3 0.3>
    overrideWorldGravity {z=-980}
   }
   n28 = Component To Local(ComponentPose=@n27)
   n29 = Get bIsCrouching()
   n30 = RigidBody(ComponentPose=Local To Component(LocalPose=Use cached pose 'Base Pose'()), OverrideWorldGravity="0,0,-980") {
    alphaInputType=Bool
    bOverrideWorldGravity=true
    overridePhysicsAsset=&/M_MED_Candor_Frontend_Cape_Physics
    simulationSpace=BaseBoneSpace
    alphaBoolBlend {blendOption=Linear}
    baseBoneRef {boneName=spine_05}
    componentAppliedLinearAccClamp=<2000 2000 2000>
    componentLinearAccScale=<0.3 0.3 0.3>
    componentLinearVelScale=<0.3 0.3 0.3>
    overrideWorldGravity {z=-980}
   }
   n31 = RigidBody(ComponentPose=@n30, OverrideWorldGravity="0,0,-980") {
    alphaInputType=Bool
    bOverrideWorldGravity=true
    overridePhysicsAsset=&/M_MED_Candor_Hood_Strap_Physics
    simulationSpace=BaseBoneSpace
    alphaBoolBlend {blendOption=Linear}
    baseBoneRef {boneName=spine_05}
    componentAppliedLinearAccClamp=<2000 2000 2000>
    componentLinearAccScale=<0.3 0.3 0.3>
    componentLinearVelScale=<0.3 0.3 0.3>
    overrideWorldGravity {z=-980}
   }
   n32 = RigidBody(ComponentPose=@n31, OverrideWorldGravity="0,0,-980") {
    alphaInputType=Bool
    bOverrideWorldGravity=true
    overridePhysicsAsset=&/M_MED_Candor_Hood_Physics
    simulationSpace=BaseBoneSpace
    alphaBoolBlend {blendOption=Linear}
    baseBoneRef {boneName=spine_05}
    componentAppliedLinearAccClamp=<2000 2000 2000>
    componentLinearAccScale=<0.3 0.3 0.3>
    componentLinearVelScale=<0.3 0.3 0.3>
    overrideWorldGravity {z=-980}
   }
   n33 = "Transform (Modify) Bone - Bone: dyn_hood_lace_01_l"(ComponentPose=@n32, Alpha=1, Translation="0,0,0", Rotation="0,15,0", Scale="1,1,1") {
    rotationMode=BMM_Additive
    rotationSpace=BCS_BoneSpace
    alphaBoolBlend {blendOption=Linear}
    boneToModify {boneName=dyn_hood_lace_01_l}
    rotation {yaw=15}
   }
   n34 = "Transform (Modify) Bone - Bone: dyn_cape_mid_r_base"!(
    ComponentPose=@n33
    bAlphaBoolEnabled:bool=NOT Boolean(A=Get bIsPlayingEmote())
    Translation="0,0,0"
    Rotation="10,10.000,-0.000"
    Scale="1,1,1"
   ) -> (Pose) {
    alphaInputType=Bool
    rotationMode=BMM_Additive
    rotationSpace=BCS_BoneSpace
    translationMode=BMM_Additive
    translationSpace=BCS_BoneSpace
    alphaBoolBlend {blendOption=Linear}
    boneToModify {boneName=dyn_cape_mid_r_base}
    rotation {pitch=10, yaw=10}
   }
   Save cached pose 'RBAN'(Pose=Blend Poses by bool(
     BlendPose_0=Component To Local(ComponentPose=@n34)
     BlendPose_1=Blend Poses by bool(BlendPose_0=@n28, BlendPose_1=@n26, BlendTime_0=0.25, BlendTime_1=0.25, bActiveValue=@n29) {blendType=Linear}
     BlendTime_0=0.25
     BlendTime_1=0.25
     bActiveValue=Get bIsInFrontEnd()
    ) {blendType=Linear}) {cacheName=RBAN}
   n35 = Control Rig!(Source=Copy Pose From Mesh() {bCopyCurves=true, bUseAttachedParent=true}, Alpha) -> (Pose) {
    alphaBoolBlend {blendOption=Linear}
    controlRigAssetReference {blueprintRigClass=&M_MED_Candor_DeformRig_ControlRig_C}
   }
   n36 = "Constraint - dyn_cape_mid_r_base (PP)"(ComponentPose=Local To Component(LocalPose=@n35), Alpha=1, ConstraintWeights_0=0.4, ConstraintWeights_1=0.6) {
    alphaBoolBlend {blendOption=Linear}
    boneToModify {boneName=dyn_cape_mid_r_base}
    constraintSetup [{targetBone {boneName=upperarm_r}}, {targetBone {boneName=clavicle_r}}]
    constraintWeights [1,1]
   }
   n37 = "Constraint - dyn_cape_mid_l_base (PP)"(ComponentPose=@n36, Alpha=1, ConstraintWeights_0=0.4, ConstraintWeights_1=0.6) {
    alphaBoolBlend {blendOption=Linear}
    boneToModify {boneName=dyn_cape_mid_l_base}
    constraintSetup [{targetBone {boneName=upperarm_l}}, {targetBone {boneName=clavicle_l}}]
    constraintWeights [1,1]
   }
   n38 = "Constraint - dyn_goggles_base (PPPPP)"(
    ComponentPose=@n37
    Alpha=1
    ConstraintWeights_0=0.15
    ConstraintWeights_1=0.3
    ConstraintWeights_2=0.2
    ConstraintWeights_3=0.125
    ConstraintWeights_4=0.225
   ) {
    alphaBoolBlend {blendOption=Linear}
    boneToModify {boneName=dyn_goggles_base}
    constraintSetup [
     {perAxis {bX=false, bY=false}, targetBone {boneName=spine_02}}
     {perAxis {bX=false, bY=false}, targetBone {boneName=spine_03}}
     {perAxis {bX=false, bY=false}, targetBone {boneName=spine_04}}
     {targetBone {boneName=clavicle_l}}
     {targetBone {boneName=deform_pec_l}}
    ]
    constraintWeights [1,1,1,1,1]
   }
   n39 = "Transform (Modify) Bone - Bone: dyn_hood"(ComponentPose=@n38, Alpha=Get headAlpha(), Translation="0,0,0", Rotation="15,-20,0.000", Scale="1,1,1") {
    rotationMode=BMM_Additive
    rotationSpace=BCS_BoneSpace
    alphaBoolBlend {blendOption=Linear}
    alphaScaleBiasClamp {bInterpResult=true, interpSpeedIncreasing=20}
    boneToModify {boneName=dyn_hood}
    rotation {pitch=15, yaw=-20}
   }
   n40 = "Transform (Modify) Bone - Bone: dyn_cape_r_base"(ComponentPose=@n39, Alpha=1, Translation="0,0,0", Rotation="-20,-15,0", Scale="1,1,1") {
    rotationMode=BMM_Additive
    rotationSpace=BCS_BoneSpace
    alphaBoolBlend {blendOption=Linear}
    boneToModify {boneName=dyn_cape_r_base}
    rotation {pitch=-20, yaw=-15}
   }
   n41 = "Transform (Modify) Bone - Bone: dyn_cape_mid_r_base"!(ComponentPose=@n40, Alpha:float=1, Translation="0,0,0", Rotation="0.000,-15,0.000", Scale="1,1,1") -> (Pose) {
    rotationMode=BMM_Additive
    rotationSpace=BCS_BoneSpace
    alphaBoolBlend {blendOption=Linear}
    boneToModify {boneName=dyn_cape_mid_r_base}
    rotation {yaw=-15}
   }
   n42 = "Transform (Modify) Bone - Bone: dyn_cape_back_r_base"(ComponentPose=@n41, Alpha=1, Translation="0,0,0", Rotation="0,-15,0", Scale="1,1,1") {
    rotationMode=BMM_Additive
    rotationSpace=BCS_BoneSpace
    alphaBoolBlend {blendOption=Linear}
    boneToModify {boneName=dyn_cape_back_r_base}
    rotation {yaw=-15}
   }
   n43 = "Transform (Modify) Bone - Bone: dyn_cape_l_base"(ComponentPose=@n42, Alpha=1, Translation="0,0,0", Rotation="20,-15,0", Scale="1,1,1") {
    rotationMode=BMM_Additive
    rotationSpace=BCS_BoneSpace
    alphaBoolBlend {blendOption=Linear}
    boneToModify {boneName=dyn_cape_l_base}
    rotation {pitch=20, yaw=-15}
   }
   n44 = "Transform (Modify) Bone - Bone: dyn_cape_mid_l_base"(ComponentPose=@n43, Alpha=1, Translation="0,0,0", Rotation="0,-15,0", Scale="1,1,1") {
    rotationMode=BMM_Additive
    rotationSpace=BCS_BoneSpace
    alphaBoolBlend {blendOption=Linear}
    boneToModify {boneName=dyn_cape_mid_l_base}
    rotation {yaw=-15}
   }
   n45 = "Transform (Modify) Bone - Bone: dyn_cape_back_l_base"(ComponentPose=@n44, Alpha=1, Translation="0,0,0", Rotation="0,-15,0", Scale="1,1,1") {
    rotationMode=BMM_Additive
    rotationSpace=BCS_BoneSpace
    alphaBoolBlend {blendOption=Linear}
    boneToModify {boneName=dyn_cape_back_l_base}
    rotation {yaw=-15}
   }
   n46 = "Transform (Modify) Bone - Bone: dyn_cape_l_base"(ComponentPose=@n45, Alpha=Get upperArmLTwistAlpha() -> (upperArmLTwistAlpha:double), Translation="0,0,0", Rotation="0,-45,0", Scale="1,1,1") {
    rotationMode=BMM_Additive
    rotationSpace=BCS_BoneSpace
    alphaBoolBlend {blendOption=Linear}
    boneToModify {boneName=dyn_cape_l_base}
    rotation {yaw=-45}
   }
   n47 = Get upperArmRUpAlpha() -> (upperArmRUpAlpha:double)
   n48 = "Transform (Modify) Bone - Bone: dyn_cape_mid_r_base"(ComponentPose=@n46, Alpha=@n47, Translation="0,0,0", Rotation="0,-30,0", Scale="1,1,1") {
    rotationMode=BMM_Additive
    rotationSpace=BCS_BoneSpace
    alphaBoolBlend {blendOption=Linear}
    boneToModify {boneName=dyn_cape_mid_r_base}
    rotation {yaw=-30}
   }
   n49 = "Transform (Modify) Bone - Bone: dyn_cape_r_base"(ComponentPose=@n48, Alpha=@n47, Translation="0,0,0", Rotation="20,-30,0", Scale="1,1,1") {
    rotationMode=BMM_Additive
    rotationSpace=BCS_BoneSpace
    alphaBoolBlend {blendOption=Linear}
    boneToModify {boneName=dyn_cape_r_base}
    rotation {pitch=20, yaw=-30}
   }
   Save cached pose 'RBAN Disabled'(Pose=Component To Local(ComponentPose=@n49)) {cacheName=RBAN Disabled}
  ]}
 EventGraph {nodes [
   Event BlueprintUpdateAnimation()
   > n1 = Sequence()
   > n2 = Gravity Override State Function DEPRECATED(CharacterPartAnimInstance=Self-Reference(), GameTimeInSeconds=GetGameTimeInSeconds(), GravityOverrideParams=Get Gravity Override Params())
   > Set GravityOverride(GravityOverride=@n2)
   > Set isWearingBackbling(isWearingBackbling=CheckForBackbling())
   > Set frontend_emote(frontend_emote:bool=AND Boolean(A=Get bIsInFrontEnd(), B=Get bIsPlayingEmote())) -> (Output_Get:bool)
   n3 = BreakTransform(InTransform=GetDeltaTransformFromRefPose(self=GetOwnerSkeletalMeshComponent(), BoneName=head, BaseName=spine_05))
   Set headAlpha(execute=@n1.then_1, headAlpha:double=MapRangeClamped(Value=BreakRotator(InRot=@n3.Rotation).Yaw, InRangeA=-15, InRangeB=-45)) -> (Output_Get:double)
   n4 = BreakTransform(InTransform=GetDeltaTransformFromRefPose(self=GetOwnerSkeletalMeshComponent(), BoneName=upperarm_r, BaseName=clavicle_r))
   > Set upperArmLTwistAlpha(upperArmLTwistAlpha:double=MapRangeClamped(Value=BreakRotator(InRot=@n4.Rotation), InRangeA=0, InRangeB=15)) -> (Output_Get:double)
   n5 = BreakTransform(InTransform=GetDeltaTransformFromRefPose(self=GetOwnerSkeletalMeshComponent(), BoneName=upperarm_l, BaseName=clavicle_l))
   > Set upperArmRUpAlpha(upperArmRUpAlpha:double=MapRangeClamped(Value=BreakRotator(InRot=@n5.Rotation).Yaw, InRangeA=0, InRangeB=90)) -> (Output_Get:double)
  ]}
]
variables [
 frontend_emote {type=boolean}
 Gravity Override Params {type=GravityOverrideParamsStruct}
 GravityOverride {type=Vector}
 headAlpha {type=number}
 isWearingBackbling {type=boolean}
 upperArmLTwistAlpha {type=number}
 upperArmRUpAlpha {type=number}
]
