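# Physics asset for the face accessory mesh of the male medium Candor player body, defining simulation bodies for its hair bones.
# Seven short hair chains of two to four uniform capsule bodies each hang from a shared hair_base capsule.
# Each link is constrained to its parent with tight swing and twist limits, with swing range varying per chain so some strands move more freely than others.
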PhysicsAsset /BRCosmetics/Characters/Player/Male/Medium/Bodies/M_MED_Candor/Meshes/Parts/M_MED_Candor_FaceAcc_Physics
bodies {
 hair_01_1 [{shapeType=Capsule, length=2, radius=1, shapeName=hair_01_1_capsule, center=<1 0 0>, rotation=<rot 90 0 0>}]
 hair_01_2 [{shapeType=Capsule, length=2, radius=1, shapeName=hair_01_2_capsule, center=<1 0 0>, rotation=<rot 90 0 0>}]
 hair_01_3 [{shapeType=Capsule, length=2, radius=1, shapeName=hair_01_3_capsule, center=<1 0 0>, rotation=<rot 90 0 0>}]
 hair_02_1 [{shapeType=Capsule, length=2, radius=1, shapeName=hair_02_1_capsule, center=<1 0 0>, rotation=<rot 90 0 0>}]
 hair_02_2 [{shapeType=Capsule, length=2, radius=1, shapeName=hair_02_2_capsule, center=<1 0 0>, rotation=<rot 90 0 0>}]
 hair_02_3 [{shapeType=Capsule, length=2, radius=1, shapeName=hair_02_3_capsule, center=<1 0 0>, rotation=<rot 90 0 0>}]
 hair_03_1 [{shapeType=Capsule, length=2, radius=1, shapeName=hair_03_1_capsule, center=<1 0 0>, rotation=<rot 90 0 0>}]
 hair_03_2 [{shapeType=Capsule, length=2, radius=1, shapeName=hair_03_2_capsule, center=<1 0 0>, rotation=<rot 90 0 0>}]
 hair_03_3 [{shapeType=Capsule, length=2, radius=1, shapeName=hair_03_3_capsule, center=<1 0 0>, rotation=<rot 90 0 0>}]
 hair_03_4 [{shapeType=Capsule, length=2, radius=1, shapeName=hair_03_4_capsule, center=<1 0 0>, rotation=<rot 90 0 0>}]
 hair_04_1 [{shapeType=Capsule, length=2, radius=1, shapeName=hair_04_1_capsule, center=<1 0 0>, rotation=<rot 90 0 0>}]
 hair_04_2 [{shapeType=Capsule, length=2, radius=1, shapeName=hair_04_2_capsule, center=<1 0 0>, rotation=<rot 90 0 0>}]
 hair_04_3 [{shapeType=Capsule, length=2, radius=1, shapeName=hair_04_3_capsule, center=<1 0 0>, rotation=<rot 90 0 0>}]
 hair_05_1 [{shapeType=Capsule, length=2, radius=1, shapeName=hair_05_1_capsule, center=<1 0 0>, rotation=<rot 90 0 0>}]
 hair_05_2 [{shapeType=Capsule, length=2, radius=1, shapeName=hair_05_2_capsule, center=<1 0 0>, rotation=<rot 90 0 0>}]
 hair_05_3 [{shapeType=Capsule, length=2, radius=1, shapeName=hair_05_3_capsule, center=<1 0 0>, rotation=<rot 90 0 0>}]
 hair_06_1 [{shapeType=Capsule, length=2, radius=1, shapeName=hair_06_1_capsule, center=<1 0 0>, rotation=<rot 90 0 0>}]
 hair_06_2 [{shapeType=Capsule, length=2, radius=1, shapeName=hair_06_2_capsule, center=<1 0 0>, rotation=<rot 90 0 0>}]
 hair_06_3 [{shapeType=Capsule, length=2, radius=1, shapeName=hair_06_3_capsule, center=<1 0 0>, rotation=<rot 90 0 0>}]
 hair_07_1 [{shapeType=Capsule, length=2, radius=1, shapeName=hair_07_1_capsule, center=<1 0 0>, rotation=<rot 90 0 0>}]
 hair_07_2 [{shapeType=Capsule, length=2, radius=1, shapeName=hair_07_2_capsule, center=<1 0 0>, rotation=<rot 90 0 0>}]
 hair_base [{shapeType=Capsule, length=1, radius=1}]
}
constraints [
 {bone1Name=hair_01_1, bone2Name=hair_base, swing1LimitDegrees=10, swing2LimitDegrees=10, twistLimitDegrees=5}
 {bone1Name=hair_01_2, bone2Name=hair_01_1, swing1LimitDegrees=10, swing2LimitDegrees=10, twistLimitDegrees=5}
 {bone1Name=hair_01_3, bone2Name=hair_01_2, swing1LimitDegrees=10, swing2LimitDegrees=10, twistLimitDegrees=5}
 {bone1Name=hair_02_1, bone2Name=hair_base, swing1LimitDegrees=5, swing2LimitDegrees=5, twistLimitDegrees=5}
 {bone1Name=hair_02_2, bone2Name=hair_02_1, swing1LimitDegrees=5, swing2LimitDegrees=5, twistLimitDegrees=5}
 {bone1Name=hair_02_3, bone2Name=hair_02_2, swing1LimitDegrees=5, swing2LimitDegrees=5, twistLimitDegrees=5}
 {bone1Name=hair_03_1, bone2Name=hair_base, swing1LimitDegrees=8, swing2LimitDegrees=8, twistLimitDegrees=5}
 {bone1Name=hair_03_2, bone2Name=hair_03_1, swing1LimitDegrees=8, swing2LimitDegrees=8, twistLimitDegrees=5}
 {bone1Name=hair_03_3, bone2Name=hair_03_2, swing1LimitDegrees=8, swing2LimitDegrees=8, twistLimitDegrees=5}
 {bone1Name=hair_03_4, bone2Name=hair_03_3, swing1LimitDegrees=8, swing2LimitDegrees=8, twistLimitDegrees=5}
 {bone1Name=hair_04_1, bone2Name=hair_base, swing1LimitDegrees=5, swing2LimitDegrees=5, twistLimitDegrees=5}
 {bone1Name=hair_04_2, bone2Name=hair_04_1, swing1LimitDegrees=5, swing2LimitDegrees=5, twistLimitDegrees=5}
 {bone1Name=hair_04_3, bone2Name=hair_04_2, swing1LimitDegrees=5, swing2LimitDegrees=5, twistLimitDegrees=5}
 {bone1Name=hair_05_1, bone2Name=hair_base, swing1LimitDegrees=5, swing2LimitDegrees=5, twistLimitDegrees=5}
 {bone1Name=hair_05_2, bone2Name=hair_05_1, swing1LimitDegrees=5, swing2LimitDegrees=5, twistLimitDegrees=5}
 {bone1Name=hair_05_3, bone2Name=hair_05_2, swing1LimitDegrees=5, swing2LimitDegrees=5, twistLimitDegrees=5}
 {bone1Name=hair_06_1, bone2Name=hair_base, swing1LimitDegrees=10, swing2LimitDegrees=10, twistLimitDegrees=5}
 {bone1Name=hair_06_2, bone2Name=hair_06_1, swing1LimitDegrees=10, swing2LimitDegrees=10, twistLimitDegrees=5}
 {bone1Name=hair_06_3, bone2Name=hair_06_2, swing1LimitDegrees=10, swing2LimitDegrees=10, twistLimitDegrees=5}
 {bone1Name=hair_07_1, bone2Name=hair_base, swing1LimitDegrees=5, swing2LimitDegrees=5, twistLimitDegrees=5}
 {bone1Name=hair_07_2, bone2Name=hair_07_1, swing1LimitDegrees=5, swing2LimitDegrees=5, twistLimitDegrees=5}
]
